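# Material function for the Octane UmbraGenesis LTE01 car skin, setting base color, metallic, roughness, emissive and normal through named reroutes.
# Base color layers anisotropic noise streaks, a trim-sheet logo and sweeping rip masks that composite character portraits.
# A time value scaled by CharacterShuffleSpeed drives the rips and shuffles portrait cells from a sprite sheet.
# Rotated, partly mirrored samples of a packed texture add shapes with animated fractional banding.
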MaterialFunction /VehicleCosmetics/Bodies/Octane/Skins/Octane_UmbraGenesisLTE01/MF_Skin_Starter
graph {nodes [
  n1 = NamedRerouteUsage()
  n2 = Constant() {r=1.6}
  n3 = Add(A=Multiply(A=Rotator(Coordinate=@n1, Time=@n2) {speed=1}) {constB=1.5}, B=Constant2Vector() {g=-0.3, r=-0.15})
  n4 = Multiply(A=NamedRerouteUsage(), B=ScalarParameter() {defaultValue=0.25, parameterName=CharacterShuffleSpeed})
  n5 = Multiply(A=Sine(None=Subtract(A=@n4) {constB=0.8})) {constB=-1}
  n6 = Saturate(None=Multiply(A=Add(A=@n5) {constB=0.7}) {constB=30})
  n7 = NamedRerouteUsage()
  n8 = Sine(None=@n7) {period=0.2}
  n9 = Sine(None=@n7) {period=0.15}
  n10 = Multiply(A=@n6, B=Multiply(A=@n8, B=@n9))
  n11 = AppendVector(A=@n8, B=@n9)
  Add(A=Rotator(Coordinate=@n3, Time=Multiply(A=@n10) {constB=0.2}) {speed=1}, B=Multiply(A=Multiply(A=@n11, B=@n6)) {constB=0.05})
  n12 = SetMaterialAttributes!(
   MaterialAttributes
   BaseColor=NamedRerouteUsage()
   Metallic=LinearInterpolate(Alpha=NamedRerouteUsage())
   Roughness=LinearInterpolate(Alpha=NamedRerouteUsage())
   EmissiveColor=NamedRerouteUsage()
   Normal=NamedRerouteUsage()
   AmbientOcclusion
  ) {attributeSetTypes [
    #1
    #2
    #3
    #4
    #5
    #6
   ]}
  FunctionOutput(None=@n12)
  NamedRerouteDeclaration(None=Constant()) {name=MetalMask, nodeColor=<rgba 0 1 0.4141 1>}
  NamedRerouteDeclaration(None=Constant3Vector() {constant=<rgba 0.8963 0 0.0018 1>}) {name=TeamColor, nodeColor=<rgba 0.003 0.0179 1 1>}
  NamedRerouteDeclaration(None=Constant3Vector() {constant=<rgba 0.0052 0.0052 0.0052 1>}) {name=CustomColor, nodeColor=<rgba 0 1 0.0625 1>}
  NamedRerouteDeclaration(None=MaterialFunctionCall!(Fresnel Power=Constant() {r=3}, Normal=NamedRerouteUsage()) {materialFunction=/VehicleCosmetics/Materials/MaterialFunctions/MF_Fresnel_Prime}.Result) {name=Fresnel, nodeColor=<rgba 0.7344 0 1 1>}
  NamedRerouteDeclaration(None=Time() {bOverride_Period=true, period=500}) {name=time, nodeColor=<rgba 1 0.9285 0 1>}
  n13 = NamedRerouteUsage()
  n14 = LinearInterpolate(A=NamedRerouteUsage(), B=@n13, Alpha=NamedRerouteUsage()) {constB=0}
  n15 = MaterialFunctionCall!(
   Texture Coordinates=Add(A=TextureCoordinate() {coordinateIndex=1}) {constB=0.1}
   Tiling X=Constant() {r=6}
   Tiling Y=Constant() {r=0.2}
   PanSpeedX
   PanSpeedY=Constant() {r=0.5}
   RotateSpeed
   Angle
   Chaos=Constant() {r=2}
   ChaosSpeed
  ) {materialFunction=/VehicleCosmetics/Materials/MaterialFunctions/MF_Noisemaker_Anisotropic}
  n16 = TextureCoordinate() {coordinateIndex=1}
  n17 = Subtract(A=@n16) {constB=0.5}
  n18 = Abs(None=ComponentMask(None=@n17) {r=true})
  n19 = ComponentMask(None=@n17) {g=true}
  n20 = Add(A=@n18, B=Multiply(A=@n18, B=Multiply(A=@n19) {constB=-0.5}))
  n21 = Rotator(
   Coordinate=AppendVector(A=Power(Base=@n20) {constExponent=0.12}, B=@n19)
   Time=Divide(A=Constant() {r=-45}, B=Constant() {desc=RadianRemover, r=57.2958})
  )
  n22 = Constant() {r=25}
  n23 = Constant() {r=0.1}
  n24 = Constant()
  n25 = Constant() {r=0.2}
  n26 = Constant()
  n27 = Constant() {r=1}
  n28 = Saturate(None=Ceil(None=Subtract(A=@n18) {constB=0.03}))
  n29 = Add(A=LinearInterpolate(
    A=Add(A=@n15.Aniso1) {constB=0.5}
    B=MaterialFunctionCall(Texture Coordinates=@n21, Tiling X=@n22, Tiling Y=@n23, PanSpeedX=@n24, PanSpeedY=@n25, Angle=@n26, Chaos=@n27) {materialFunction=&MF_Noisemaker_Anisotropic}.Aniso1
    Alpha=@n28
   )) {constB=-0.3}
  n30 = Constant() {r=50}
  n31 = Constant() {r=0.2}
  n32 = Constant() {r=0.3}
  n33 = Constant()
  n34 = Constant() {r=0.3}
  n35 = Add(A=Multiply(
    A=@n28
    B=MaterialFunctionCall(Texture Coordinates=@n21, Tiling X=@n30, Tiling Y=@n31, PanSpeedY=@n32, Angle=@n33, Chaos=@n34) {materialFunction=&MF_Noisemaker_Anisotropic}.Aniso1
   )) {constB=-0.2}
  n36 = Subtract(A=@n16, B=Constant2Vector() {g=1, r=0.5}) {constB=0.5}
  n37 = Multiply(A=Add(A=Multiply(A=@n29) {constB=20}, B=Multiply(A=@n35) {constB=15}), B=OneMinus(None=DotProduct(A=@n36, B=@n36)))
  n38 = Multiply(A=@n36, B=Constant2Vector() {g=7, r=10}) {constB=10}
  n39 = OneMinus(None=DotProduct(A=@n38, B=@n38))
  n40 = Saturate(None=ConstantBiasScale(None=Add(A=@n37, B=@n39)) {bias=-0.5, scale=60})
  n41 = TextureObjectParameter() {
   desc=RGB +Alpha opacity
   group=UniversalLogos
   parameterName=Logo
   texture=&/T_UmbraGenesis_Logo
  }
  n42 = VectorParameter() {
   desc="Number of sprites in X,Y=r,g  Select Sprite position starting from top left in X,Y = b,a"
   group=UniversalLogos
   parameterName=TrimSheet1_Cells
   channelNames {a=Sprite Y Position, b=Sprite X Position, g="##Sprites in Y", r="##Sprites in X"}
   defaultValue=<rgba 1 1 1 1>
  }
  n43 = VectorParameter() {
   desc="RG=XY Scale Offset; BA=XY Position Offset"
   group=UniversalLogos
   parameterName=TrimSheet1Mirrored_Offset
   channelNames {a=Y Offset, b=X Offset, g=Y Scale, r=X Scale}
  }
  n44 = MaterialFunctionCall!(
   TrimSheet=@n41
   TrimSheet_Transform=Constant4Vector() {constant=<rgba 1.75 1.65 0 0.75>}.RGBA
   TrimSheet_Rotation=ScalarParameter() {desc=Degrees counterclockwise, group=UniversalLogos, parameterName=TrimSheet1_Rotation}
   TrimSheet_Cells=AppendVector(A=@n42."##Sprites in X", B=@n42."##Sprites in X")
   TrimSheet_CellSelector=AppendVector(A=@n42.Sprite X Position, B=@n42.Sprite X Position)
   MirroredTrimSheet_Opacity=ScalarParameter() {desc="Enables the mirrored TrimSheet; 0 = hidden; 1 = visible", group=UniversalLogos, parameterName=TrimSheet1Mirrored_Opacity}
   MirroredTrimSheet_TransformOffset=AppendVector(A=@n43, B=@n43)
   MirroredTrimSheet_RotationOffset=ScalarParameter() {desc=Offset by degrees counterclockwise, group=UniversalLogos, parameterName=TrimSheet1Mirrored_RotationOffset}
   MirroredTrimSheet_CellSelector
   MirrorUsesFlippedUVs
  ) {materialFunction=/VehicleCosmetics/AutoUniversalSkins/_MaterialFunctions/MF_AutomatedUniCoordSelector}
  n45 = Subtract(
   A=ComponentMask(None=@n44.TextureOutput) {a=true}
   B=Ceil(None=TextureSample() {texture=&/T_UmbraGenesis_RGB1}.A)
  )
  n46 = NamedRerouteUsage()
  n47 = Subtract(A=TextureCoordinate() {coordinateIndex=1}) {constB=0.5}
  n48 = Abs(None=ComponentMask(None=@n47) {r=true})
  n49 = AppendVector(A=@n48, B=ComponentMask(None=@n47) {g=true})
  n50 = Constant() {r=7}
  n51 = Constant() {desc=RadianRemover, r=57.2958}
  n52 = ComponentMask(None=Rotator(Coordinate=@n49, Time=Divide(A=@n50, B=@n51))) {r=true}
  n53 = Subtract(A=Saturate(None=ConstantBiasScale(None=@n48) {bias=-0.17, scale=5}), B=Saturate(None=ConstantBiasScale(None=@n52) {bias=-0.275, scale=10})) {constB=0.3}
  n54 = Add(A=@n4) {constB=0.5}
  n55 = Rotator(Coordinate=@n49, Time=Divide(A=Constant() {r=60}, B=@n51))
  n56 = Multiply(A=Add(A=Multiply(A=@n54) {constB=-4}, B=ComponentMask(None=@n55) {r=true})) {constB=0.25}
  n57 = Floor(None=@n56)
  n58 = Abs(None=ConstantBiasScale(None=Subtract(A=@n56, B=@n57)) {bias=-0.5, scale=1}) {desc=Sweep}
  n59 = ConstantBiasScale(None=@n58) {bias=-0.14, scale=100}
  n60 = TextureSample(UVs=BumpOffset(Coordinate=TextureCoordinate() {coordinateIndex=1}, Height=Constant() {r=0.75})) {texture=&T_UmbraGenesis_RGB1}
  n61 = Add(A=Multiply(A=Saturate(None=@n53), B=Saturate(None=@n59)) {desc=RipsMotion}, B=Multiply(A=@n60.G) {constB=0.25}) {constB=0}
  n62 = Multiply(A=@n60.G) {constB=0.15}
  n63 = ConstantBiasScale(None=@n48) {bias=-0.17, scale=2}
  n64 = ConstantBiasScale(None=@n48) {bias=-0.23, scale=4}
  n65 = Saturate(None=Subtract(A=Saturate(None=@n63), B=Saturate(None=@n64)) {constB=0.3})
  n66 = Add(A=@n4) {constB=0.5013}
  n67 = Rotator(Coordinate=@n49, Time=Divide(A=Constant() {r=30}, B=@n51))
  n68 = Multiply(A=Add(A=Multiply(A=@n66) {constB=-4}, B=ComponentMask(None=@n67) {r=true})) {constB=0.25}
  n69 = ConstantBiasScale(None=Subtract(A=@n68, B=Floor(None=@n68))) {bias=-0.5, scale=1}
  n70 = Saturate(None=ConstantBiasScale(None=Abs(None=@n69) {desc=Sweep}) {bias=-0.1, scale=20})
  n71 = Ceil(None=Multiply(A=@n65, B=@n70))
  n72 = Add(A=Saturate(None=Multiply(A=@n62, B=@n71)), B=Multiply(A=Multiply(A=@n65), B=@n70) {desc=RipsMotion}) {constB=0}
  n73 = Saturate(None=Max(A=ConstantBiasScale(None=@n61) {bias=-0.15, scale=500}, B=ConstantBiasScale(None=@n72) {bias=-0.125, scale=1000}))
  n74 = NamedRerouteUsage()
  n75 = TextureObjectParameter() {
   desc=RGB +Alpha opacity
   group=UniversalLogos
   parameterName=CharacterPortraits
   texture=&/T_UmbraGenesis_Portraits
  }
  n76 = Constant() {desc=scale, r=6}
  n77 = Constant2Vector() {desc=offset, g=-0.1, r=0.45}
  n78 = Constant() {r=4}
  n79 = Fmod(A=@n4, B=@n78)
  n80 = Ceil(None=@n79)
  n81 = Multiply(A=Saturate(None=Subtract(A=@n80, B=@n79)), B=Constant2Vector() {g=-0.05}) {constB=0.02}
  n82 = Constant() {r=4}
  n83 = Fmod(A=Divide(A=@n4, B=@n78), B=@n78)
  n84 = MaterialFunctionCall(
   TrimSheet=@n75
   TrimSheet_Transform=AppendVector(A=AppendVector(A=@n76, B=@n76), B=Add(A=@n77, B=@n81))
   TrimSheet_Rotation=Constant() {r=95}
   TrimSheet_Cells=AppendVector(A=@n82, B=@n82)
   TrimSheet_CellSelector=AppendVector(A=Ceil(None=@n83), B=@n80) {desc=Shuffle}
   MirroredTrimSheet_Opacity=Constant() {r=1}
   MirroredTrimSheet_TransformOffset=Constant4Vector() {constant=<rgba 0 0 0 0.2>}.RGBA
   MirroredTrimSheet_RotationOffset=Constant()
   MirrorUsesFlippedUVs=StaticBool()
  ) {materialFunction=&MF_AutomatedUniCoordSelector}
  n85 = ComponentMask(None=@n84.TextureOutput) {b=true, g=true, r=true}
  n86 = ComponentMask(None=@n84.TextureOutput) {a=true}
  n87 = ConstantBiasScale(None=@n61) {bias=-0.2, scale=500}
  n88 = Multiply(A=Saturate(None=ConstantBiasScale(None=@n48) {bias=-0.1, scale=5}), B=ConstantBiasScale(None=@n58) {bias=-0.14, scale=100})
  n89 = ConstantBiasScale(None=Multiply(A=@n61, B=Saturate(None=@n88))) {bias=-0.1, scale=500}
  n90 = ConstantBiasScale(None=@n72) {bias=-0.1, scale=1000}
  n91 = LinearInterpolate(
   A=LinearInterpolate(
    A=LinearInterpolate(A=@n14, B=@n13, Alpha=@n40)
    B=ComponentMask(None=@n44.TextureOutput) {b=true, g=true, r=true}
    Alpha=Saturate(None=@n45)
   )
   B=LinearInterpolate(A=LinearInterpolate(A=@n46, Alpha=@n73), B=LinearInterpolate(A=@n74, B=@n85, Alpha=@n86), Alpha=Saturate(None=@n87))
   Alpha=Saturate(None=Max(A=@n89, B=@n90))
  )
  n92 = TextureSample(UVs=@n3) {texture=&T_UmbraGenesis_RGB1}
  n93 = LinearInterpolate(A=NamedRerouteUsage(), Alpha=Saturate(None=ConstantBiasScale(None=@n92.R) {bias=-0.51, scale=4})) {constB=0.8}
  n94 = Saturate(None=ConstantBiasScale(None=@n92.R) {bias=-0.01, scale=10})
  n95 = NamedRerouteUsage()
  n96 = TextureCoordinate() {coordinateIndex=1}
  n97 = Add(A=Multiply(A=@n96, B=Constant() {desc=size, r=3}) {constB=3}, B=Constant2Vector() {g=0.25, r=-1.5})
  n98 = Constant() {r=2}
  n99 = Add(A=Multiply(A=@n96, B=Constant() {desc=size, r=3}) {constB=3}, B=Constant2Vector() {g=0.1, r=-0.5})
  n100 = Rotator(Coordinate=@n99, Time=Constant() {r=1}) {speed=-1.5}
  n101 = Rotator(Coordinate=Add(A=Multiply(A=@n99) {constB=1.25}, B=Constant2Vector() {r=-0.5}), Time=Constant()) {speed=1}
  n102 = Max(
   A=TextureSample(UVs=Rotator(Coordinate=@n97, Time=@n98) {speed=-1.5}) {texture=&T_UmbraGenesis_RGB1}.B
   B=Max(
    A=TextureSample(UVs=@n100) {texture=&T_UmbraGenesis_RGB1}.B
    B=TextureSample(UVs=@n101) {texture=&T_UmbraGenesis_RGB1}.B
   )
  )
  n103 = Multiply(A=Add(A=Multiply(A=@n95) {constB=0.5}, B=@n102)) {constB=4}
  n104 = Floor(None=@n103)
  n105 = Saturate(None=ConstantBiasScale(None=Subtract(A=@n103, B=@n104)) {bias=-0.2, scale=5})
  n106 = Saturate(None=ConstantBiasScale(None=@n102) {bias=-0.05, scale=50})
  n107 = ConstantBiasScale(None=@n102) {bias=-0.95, scale=50}
  n108 = LinearInterpolate(A=NamedRerouteUsage(), B=NamedRerouteUsage(), Alpha=Max(A=Multiply(A=@n105, B=@n106), B=Saturate(None=@n107)))
  n109 = NamedRerouteUsage()
  n110 = ConstantBiasScale(None=ComponentMask(None=@n96) {r=true}) {bias=-0.5, scale=2}
  n111 = ComponentMask(None=@n96) {g=true}
  n112 = NamedRerouteDeclaration(None=AppendVector(A=Abs(None=@n110), B=Multiply(B=@n111) {constA=2})) {name=symmetricalUVs, nodeColor=<rgba 1 0 0.5625 1>}
  n113 = Constant() {desc=size, r=3}
  n114 = Constant()
  n115 = Rotator(Coordinate=Add(A=Multiply(A=@n112, B=@n113) {constB=3}, B=Constant2Vector() {g=-0.5, r=-1.7}), Time=@n114) {speed=-1.5}
  n116 = Add(A=Multiply(A=NamedRerouteUsage(), B=Constant() {desc=size, r=3}) {constB=3}, B=Constant2Vector() {g=-1, r=-1.4})
  n117 = Constant() {r=2}
  n118 = Rotator(Coordinate=@n116, Time=@n117) {speed=1}
  n119 = Add(A=@n116, B=Constant2Vector() {g=-0.5, r=-0.5})
  n120 = Constant() {r=1}
  n121 = TextureSample(UVs=Rotator(Coordinate=@n119, Time=@n120) {speed=1}) {texture=&T_UmbraGenesis_RGB1}
  n122 = Multiply(A=@n119, B=Constant() {desc=size, r=1.5}) {constB=3}
  n123 = Constant2Vector() {g=0.5, r=-0.2}
  n124 = TextureSample(UVs=Rotator(Coordinate=Add(A=@n122, B=@n123), Time=@n114) {speed=1}) {texture=&T_UmbraGenesis_RGB1}
  n125 = Add(A=Multiply(A=NamedRerouteUsage(), B=Constant() {desc=size, r=3}) {constB=3}, B=Constant2Vector() {g=-4, r=-0.55})
  n126 = Add(A=@n125, B=Constant2Vector() {r=-0.6})
  n127 = Add(A=@n125, B=Constant2Vector() {g=-0.5, r=-0.5})
  n128 = Max(
   A=TextureSample(UVs=Rotator(Coordinate=@n126, Time=@n117) {speed=-1}) {texture=&T_UmbraGenesis_RGB1}.B
   B=TextureSample(UVs=Rotator(Coordinate=@n127, Time=@n120) {speed=1}) {texture=&T_UmbraGenesis_RGB1}.B
  )
  n129 = Add(A=Multiply(A=@n127, B=Constant() {desc=size, r=1.5}) {constB=3}, B=Constant2Vector() {g=-0.65, r=0.6})
  n130 = Add(A=@n129, B=Constant2Vector() {g=0.3, r=-0.3})
  n131 = Max(
   A=TextureSample(UVs=Rotator(Coordinate=@n129, Time=@n114) {speed=1}) {texture=&T_UmbraGenesis_RGB1}.B
   B=TextureSample(UVs=Rotator(Coordinate=@n130, Time=@n120) {speed=-1}) {texture=&T_UmbraGenesis_RGB1}.B
  )
  n132 = Max(
   A=Max(
    A=TextureSample(UVs=@n115) {texture=&T_UmbraGenesis_RGB1}.B
    B=TextureSample(UVs=@n118) {texture=&T_UmbraGenesis_RGB1}.B
   )
   B=Max(A=Max(A=@n121.B, B=@n124.B), B=Max(A=@n128, B=@n131))
  )
  n133 = LinearInterpolate(A=NamedRerouteUsage(), B=@n132, Alpha=Saturate(None=ConstantBiasScale(None=@n132) {bias=-0.05, scale=50}))
  n134 = Multiply(A=Add(A=Multiply(A=@n109) {constB=0.5}, B=@n133)) {constB=4}
  n135 = ConstantBiasScale(None=Subtract(A=@n134, B=Floor(None=@n134))) {bias=-0.3, scale=5}
  n136 = ConstantBiasScale(None=@n133) {bias=-0.95, scale=50}
  n137 = LinearInterpolate(A=Constant3Vector() {constant=<rgba 0.8 0.8 0.8 0>}, B=NamedRerouteUsage(), Alpha=Max(A=Saturate(None=@n135), B=Saturate(None=@n136)))
  n138 = Saturate(None=ConstantBiasScale(None=@n133) {bias=-0.05, scale=50})
  n139 = Floor(None=Add(A=TextureSample() {texture=&T_UmbraGenesis_RGB1}.A) {constB=0.1})
  NamedRerouteDeclaration(None=LinearInterpolate(A=LinearInterpolate(A=@n91, B=@n93, Alpha=@n94), B=LinearInterpolate(A=@n108, B=@n137, Alpha=@n138), Alpha=Saturate(None=@n139))) {name=BaseColor, nodeColor=<rgba 1 0.0938 0 1>}
  NamedRerouteDeclaration(None=Constant3Vector() {constant=<rgba 0 0 1 0>}) {name=NormalTexture, nodeColor=<rgba 0.299 0.3454 1 1>}
  NamedRerouteDeclaration(None=Multiply(A=NamedRerouteUsage(), B=ScalarParameter() {parameterName=EmissiveIntensity})) {name=Emissive, nodeColor=<rgba 0 0.2266 1 1>}
  NamedRerouteDeclaration(None=Constant() {r=0.1}) {name=RoughnessMask, nodeColor=<rgba 0.6641 0 1 1>}
  n140 = TextureCoordinate() {coordinateIndex=1}
  n141 = Abs(None=ConstantBiasScale(None=ComponentMask(None=@n140) {r=true}) {bias=-0.5, scale=2})
  n142 = Multiply(B=ComponentMask(None=@n140) {g=true}) {constA=2}
  n143 = Multiply(
   A=NamedRerouteDeclaration(None=AppendVector(A=@n141, B=@n142)) {name=symmetricalUVs_1, nodeColor=<rgba 1 0 0.5625 1>}
   B=Constant() {desc=size, r=2.5}
  ) {constB=3}
  n144 = Constant2Vector() {g=-0.5, r=-1.7}
  n145 = Constant() {r=1}
  n146 = TextureSample(UVs=Rotator(Coordinate=Add(A=@n143, B=@n144), Time=@n145) {speed=0.5}) {texture=&T_UmbraGenesis_RGB1}
  n147 = Add(A=Multiply(A=NamedRerouteUsage(), B=Constant() {desc=size, r=3.52}) {constB=3}, B=Constant2Vector() {g=-1, r=-1.7})
  n148 = TextureSample(UVs=Rotator(Coordinate=@n147, Time=@n145) {speed=-0.75}) {texture=&T_UmbraGenesis_RGB1}
  n149 = Add(A=@n147, B=Constant2Vector() {g=-0.5, r=-0.5})
  n150 = Add(A=Multiply(A=@n149, B=Constant() {desc=size, r=1.5}) {constB=3}, B=Constant2Vector() {g=0.5, r=-0.2})
  n151 = Max(
   A=TextureSample(UVs=Rotator(Coordinate=@n149, Time=@n145) {speed=-1}) {texture=&T_UmbraGenesis_RGB1}.B
   B=TextureSample(UVs=Rotator(Coordinate=@n150, Time=@n145) {speed=1}) {texture=&T_UmbraGenesis_RGB1}.B
  )
  n152 = Add(A=Multiply(A=NamedRerouteUsage(), B=Constant() {desc=size, r=2.8}) {constB=3}, B=Constant2Vector() {g=-4, r=-0.7})
  n153 = Rotator(Coordinate=Add(A=@n152, B=Constant2Vector() {r=-0.6}), Time=@n145) {speed=-1}
  n154 = Add(A=@n152, B=Constant2Vector() {g=-0.45, r=-0.5})
  n155 = Rotator(Coordinate=@n154, Time=@n145) {speed=1}
  n156 = Add(A=Multiply(A=@n154, B=Constant() {desc=size, r=1.5}) {constB=3}, B=Constant2Vector() {g=0.35, r=0.4})
  n157 = Rotator(Coordinate=@n156, Time=@n145) {speed=1}
  n158 = Rotator(Coordinate=Add(A=@n156, B=Constant2Vector() {g=-0.6, r=-0.03}), Time=@n145) {speed=-1}
  n159 = Max(
   A=Max(
    A=TextureSample(UVs=@n153) {texture=&T_UmbraGenesis_RGB1}.B
    B=TextureSample(UVs=@n155) {texture=&T_UmbraGenesis_RGB1}.B
   )
   B=Max(
    A=TextureSample(UVs=@n157) {texture=&T_UmbraGenesis_RGB1}.B
    B=TextureSample(UVs=@n158) {texture=&T_UmbraGenesis_RGB1}.B
   )
  )
  NamedRerouteDeclaration(None=Max(A=Max(A=@n146.B, B=@n148.B), B=Max(A=@n151, B=@n159))) {name=BGStars, nodeColor=<rgba 0.6957 0 0 1>}
 ]}
